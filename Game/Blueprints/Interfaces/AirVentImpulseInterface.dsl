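; Blueprint interface for notifying an implementer that it has entered or exited an air vent, which by its naming applies an impulse.
; It declares two functions, EnteredAirVent and ExitedAirVent, with identical parameters: a launch velocity, a base force scale and a direction vector.
; The interface holds only these signatures, so any response to the vent is left to implementing classes.
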(fn EnteredAirVent
  (params (LaunchVelocity :type Float%20%28double-precision%29) (BaseForceScale :type Float%20%28double-precision%29) (Direction :type Vector)))

(fn ExitedAirVent
  (params (LaunchVelocity :type Float%20%28double-precision%29) (BaseForceScale :type Float%20%28double-precision%29) (Direction :type Vector)))
