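; BPI_HitReactionInterface is a Blueprint interface for querying the custom Fort movement modes used during knockback hit reactions.
; It declares four getter functions, one for each knockback phase: Up, Apex, Down and Land.
; Each function shows only a zero return, with no logic in the interface itself.
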(fn HitReaction_Knockback_GetCustomFortMovementMode_Apex
  (return 0))

(fn HitReaction_Knockback_GetCustomFortMovementMode_Down
  (return 0))

(fn HitReaction_Knockback_GetCustomFortMovementMode_Land
  (return 0))

(fn HitReaction_Knockback_GetCustomFortMovementMode_Up
  (return 0))
