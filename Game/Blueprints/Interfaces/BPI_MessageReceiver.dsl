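; BPI_MessageReceiver is a Blueprint Interface for objects that receive messages from other objects.
; It declares one function, Message, which takes a Name as the message and a Gameplay Tag as the message tag.
; The function has no outputs and no logic here, so each implementing Blueprint decides how it responds.
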(fn Message
  (params (Message :type Name) (MessageTag :type Gameplay%20Tag%20Structure)))
