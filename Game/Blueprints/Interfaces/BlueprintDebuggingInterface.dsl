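; Blueprint interface for debugging that exposes a GetBlueprintLogLevel function returning a log level enumeration value.
; The function body shown returns the first enumerator entry, NewEnumerator0, with no other logic.
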(fn GetBlueprintLogLevel
  (return "NewEnumerator0"))
